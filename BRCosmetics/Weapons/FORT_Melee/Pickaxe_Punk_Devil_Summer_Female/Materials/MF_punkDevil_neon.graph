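# Material function for the neon line effect on the Punk Devil Summer Female pickaxe, outputting full material attributes.
# A Blueprint-fed timestamp switches pulse speed and emissive strength from reactive to idle in a hard step once the duration elapses.
# A curve-atlas row, scrolled by time and vertex red, blends idle and bright colors.
# Other vertex channels mask color, surface values and width for spline thickening, which adds to a camera offset.
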MaterialFunction /BRCosmetics/Weapons/FORT_Melee/Pickaxe_Punk_Devil_Summer_Female/Materials/MF_punkDevil_neon
graph {nodes [
  n1 = VectorParameter() {group=PunkDevil, parameterName=PD_EmissiveIdle, defaultValue=<rgba 1.5 0.0172 0.1042 1>}
  n2 = VertexColor()
  n3 = Power(Base=@n2.A, Exp=ScalarParameter() {defaultValue=1, group=PunkDevil, parameterName=PD_EmissiveMultPow}) {constExponent=3}
  n4 = Time()
  n5 = Subtract(A=Time(), B=ScalarParameter() {defaultValue=-100, group=PunkDevil, parameterName=PD_GameTimeFromBP})
  n6 = ScalarParameter() {defaultValue=1, group=PunkDevil, parameterName=PD_ReactiveDuration}
  n7 = Saturate(None=Floor(None=Divide(A=@n5, B=@n6) {desc=the length of time the reaction takes to fade}))
  n8 = LinearInterpolate(
   A=Multiply(A=@n4, B=ScalarParameter() {defaultValue=1, group=PunkDevil, parameterName=PD_ReactiveSpeed, sortPriority=2}) {constB=1.5}
   B=Multiply(A=@n4, B=ScalarParameter() {defaultValue=0.3, group=PunkDevil, parameterName=PD_IdleSpeed, sortPriority=1}) {constB=0.3}
   Alpha=@n7
  ) {constAlpha=1}
  n9 = Multiply(A=@n2.R) {constB=0.5}
  n10 = CurveAtlasRowParameter(CurveTime=Frac(None=Subtract(A=@n8, B=@n9))) {
   atlas=&AthenaCurveAtlas01
   curve=/Game/Athena/LinearColorCurves/PunkDevilSummer_Curve
   defaultValue=227
   group=PunkDevil
   parameterName=PD_Param
  }
  n11 = Multiply(
   A=LinearInterpolate(
    A=Multiply(A=@n1.RGB, B=@n3)
    B=VectorParameter() {group=PunkDevil, parameterName=PD_EmissiveBright, defaultValue=<rgba 1 0.5 0.7 1>}.RGB
    Alpha=@n10.R
   )
   B=@n2.B
  )
  n12 = ScalarParameter() {defaultValue=8, group=PunkDevil, parameterName=PD_EmissiveMultiplierReactive}
  n13 = ScalarParameter() {defaultValue=10, group=PunkDevil, parameterName=PD_EmissiveMultiplier}
  n14 = MaterialFunctionCall(Color=Multiply(A=VectorParameter() {group=PunkDevil, parameterName=PD_MSR, defaultValue=<rgba 0 0.2 0.1 1>}.RGB, B=@n2.A)) {materialFunction=&SplitComponents}
  n15 = Multiply(A=ScalarParameter() {defaultValue=0.5, group=PunkDevil, parameterName=PD_LineWidth}, B=VertexColor().G)
  n16 = MaterialFunctionCall!(widthTip=@n15, Flatten=ScalarParameter() {group=PunkDevil, parameterName=PD_FlattenLine}, widthBase=@n15) {materialFunction=/Game/Characters/Player/Male/Medium/Bodies/M_MED_Neon/Materials/MF_splineThickenM}
  n17 = Multiply(A=Constant() {r=1})
  n18 = SetMaterialAttributes!(
   MaterialAttributes
   BaseColor=@n11
   Metallic=@n14.R
   Specular=@n14.G
   Roughness=@n14.B
   EmissiveColor=Multiply(A=@n11, B=LinearInterpolate(A=@n12, B=@n13, Alpha=@n7)) {constB=8}
   Normal=@n16.normal
   WorldPositionOffset=Add(
    A=MaterialFunctionCall(Offset Amount=@n17) {materialFunction=/Engine/Functions/Engine_MaterialFunctions02/WorldPositionOffset/CameraOffset}.Camera Offset
    B=@n16.worldPositionOffset
   )
  ) {attributeSetTypes [
    #1
    #2
    #3
    #4
    #5
    #6
    #7
   ]}
  FunctionOutput(None=@n18)
 ]}
