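# Niagara idle effect, named for the Punk Devil Summer Female backpack, that shows a pulsing pink mesh glow with an attached light.
# A single Light_And_Glow emitter burst-spawns one persistent particle on the first loop only, feeding mesh and light renderers.
# Curves indexed by System.NormalizedLoopAge animate the particle's color, alpha and light color, so the pulse follows the system loop without respawning.
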NiagaraSystem /BRCosmetics/Weapons/FORT_Melee/Pickaxe_Punk_Devil_Summer_Female/Materials/NS_Backpack_Punk_Devil_Summer_Female_Idle
dependencies {
 usedDataInterfaces [&NiagaraDataInterfaceColorCurve, &NiagaraDataInterfaceCurve]
 usedDynamicInputs []
 usedModules [&SystemState]
}
emitters {Light_And_Glow {
  emitterUpdateScript [
   &EmitterState()
   &SpawnBurst_Instantaneous(Spawn Count=1, Spawn Probability=HlslExpression() {hlslExpression="Emitter.LoopCount==0"}, Use Spawn Probability=true)
  ]
  particleSpawnScript [
   /Niagara/Modules/Spawn/Initialization/InitializeParticle(
    Color=<rgba 1 0.0267 0.1709 0.1>
    Mesh Scale=<0.5 1.1 1.1>
    Module.Write Color=true
    Module.Write Lifetime=true
    Module.Write Mass=false
    Module.Write Material Random=true
    Module.Write Position=false
    Module.Write Scale=true
    Module.Write SpriteRotation=false
    Module.Write SpriteSize=true
    Sprite Size=<65 65>
   )
   &FN_Light_v2(Light Exponent=4, Light Radius=42.5, LightEnabled=true, Position Offset:Vector3f=<10 0 0>)
  ]
  particleUpdateScript [
   &ParticleState(Kill Particles When Lifetime Has Elapsed=false, Let Infinitely Lived Particles Die When Emitter Deactivates=true)
   &ScaleColor(
    Scale Alpha=&FloatFromCurve(
     FloatCurve=DataInterface() {propertyValues {Curve {keys [
         <at 0.1 1>
         {arriveTangent=-0.7039, interpMode=RCIM_Cubic, leaveTangent=-0.7039, time=0.3472, value=1}
         <at 0.5972 0.65 RCIM_Cubic>
         <at 0.8472 1>
         <at 1.0972 0.65 RCIM_Cubic>
         <at 1.3472 1 RCIM_Cubic>
         <at 1.5972 0.65>
         <at 1.8472 1 RCIM_Cubic>
        ]}}}
     CurveIndex=$System.NormalizedLoopAge
    )
    Scale RGB=&MakeVectorFromLinearColorRGB(Color=&ColorFromCurve(
      ColorCurve=DataInterface() {propertyValues {
        AlphaCurve {keys []}
        BlueCurve {keys [<at 0 0.9245 RCIM_Cubic>,<at 0.2 1.6155 RCIM_Cubic>,<at 0.5 0.9245 RCIM_Cubic>]}
        GreenCurve {keys [<at 0 0.6562 RCIM_Cubic>,<at 0.2 0.25 RCIM_Cubic>,<at 0.5 0.6562 RCIM_Cubic>]}
        RedCurve {keys [<at 0 1 RCIM_Cubic>,<at 0.2 2 RCIM_Cubic>,<at 0.5 1 RCIM_Cubic>]}
       }}
      CurveIndex=$System.NormalizedLoopAge
     ))
    ScaleA=true
    ScaleRGB=true
   )
   /SetVariables(Particles.Light Color:LinearColor=&MultiplyLinearColor(
     Linear Color=<rgba 0.75 0.082 0.181 1>
     Scale Factor=&Vector4FromCurve(
      CurveIndex=$System.NormalizedLoopAge
      Vector4Curve=DataInterface() {propertyValues {
        AlphaCurve {keys []}
        BlueCurve {keys [<at 0 2.5125 RCIM_Cubic>,<at 0.2 4.3969 RCIM_Cubic>,<at 0.5 2.5125 RCIM_Cubic>]}
        GreenCurve {keys [<at 0 0.7813 RCIM_Cubic>,<at 0.2 1.3672 RCIM_Cubic>,<at 0.5 0.7813 RCIM_Cubic>]}
        RedCurve {keys [<at 0 3 RCIM_Cubic>,<at 0.2 5.25 RCIM_Cubic>,<at 0.5 3 RCIM_Cubic>]}
       }}
     )
    )) {isSetParameters=true}
  ]
  rendererClasses [&NiagaraLightRendererProperties, &NiagaraMeshRendererProperties]
  renderers [
   {rendererClass=&NiagaraLightRendererProperties, rendererIndex=0}
   {rendererClass=&NiagaraMeshRendererProperties, rendererIndex=1}
  ]
 }}
summary {
 systemName=NS_Backpack_Punk_Devil_Summer_Female_Idle
 emitters [{emitterName=Light_And_Glow, rendererClasses [&NiagaraLightRendererProperties, &NiagaraMeshRendererProperties]}]
 userVariables [User.SkeletalMesh {
   type=NiagaraDataInterfaceSkeletalMesh
   defaultValue {
    struct=NiagaraExt_VariableValue_DataInterface
    value {dataInterface=&self:NiagaraDataInterfaceSkeletalMesh_0}
   }
  }]
}
